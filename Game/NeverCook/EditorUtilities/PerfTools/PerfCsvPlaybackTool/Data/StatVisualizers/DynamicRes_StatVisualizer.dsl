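; Stat visualizer function for dynamic resolution in the PerfCsvPlaybackTool editor utility, mapping a stat value to a display color.
; It normalizes the value over a fixed range, apparently a resolution percentage, and uses the result to lerp from red at the low end to green at the high end.
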(fn getColorFromStatValue
  (params (value :type Float%20%28double-precision%29))
  (return (Math|Color|Lerp(LinearColor) "(R=1.000000,G=0.000000,B=0.019686,A=1.000000)" "(R=0.160958,G=0.734375,B=0.018317,A=1.000000)" (Math|Float|NormalizetoRange value 6e+01 9e+01))))
