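; GPU shadow depths stat visualizer for the perf CSV playback editor tool, mapping a stat value to a display color.
; getColorFromStatValue normalizes the value over a fixed range, applies a square-root curve, and lerps from green to red.
; The square-root curve shifts the color toward red quickly at low values, making modest shadow depth costs stand out.
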(fn getColorFromStatValue
  (params (value :type Float%20%28double-precision%29))
  (return (Math|Color|Lerp(LinearColor) "(R=0.231373,G=1.000000,B=0.094118,A=1.000000)" "(R=0.737910,G=0.072272,B=0.072272,A=1.000000)" (Math|Float|Power (Math|Float|NormalizetoRange value 0.0 5.0) 0.5))))
