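; GameThreadTime stat visualizer for the editor-only perf CSV playback tool, mapping a game thread time value to a display color.
; getColorFromStatValue normalizes the value over a 32–40 range, presumably milliseconds, and uses it to lerp from green to red.
; The graph has no clamp node, so values outside that range are not explicitly limited to the two endpoint colors.
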(fn getColorFromStatValue
  (params (value :type Float%20%28double-precision%29))
  (return (Math|Color|Lerp(LinearColor) "(R=0.232947,G=1.000000,B=0.096568,A=1.000000)" "(R=0.734375,G=0.072673,B=0.072673,A=1.000000)" (Math|Float|NormalizetoRange value 32.0 4e+01))))
