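; Base stat visualizer Blueprint for the editor-only CSV playback perf tool, named as the parent for stat-specific visualizers.
; It declares a getColorFromStatValue function that takes a double-precision stat value and is meant to map it to a display color.
; The base implementation ignores its input and returns a default zero value, leaving the actual mapping to be defined elsewhere.
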(fn getColorFromStatValue
  (params (value :type Float%20%28double-precision%29))
  (return 0))
